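# Masked subsurface-profile material for the Studded cosmetic vehicle wheel, using a jade profile.
# It extends MF_WheelParent_Simple, using rim masks limited by inverse vertex red to apply emissive tint and Fresnel base-color highlights.
# A rim texture and vertex blue cut the opacity mask, while a heightmap drives subsurface opacity.
# The result passes through vehicle damage, then de-rez transition functions.
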Material /VehicleCosmetics/Wheels/Studded/Materials/M_Wheel_Studded_SSS
graph {
 properties {
  bAutomaticallySetUsageInEditor=false
  bUseMaterialAttributes=true
  bUsedWithMeshDeformer=true
  bUsedWithSkeletalMesh=true
  blendMode=BLEND_Masked
  pixelDepthOffsetMode=PDOM_Legacy
  shadingModel=MSM_SubsurfaceProfile
  subsurfaceProfile=&/SSP_jade
  usedShadingModels=Subsurface Profile
 }
 nodes [
  n1 = MaterialFunctionCall() {materialFunction=&MF_WheelParent_Simple}
  n2 = GetMaterialAttributes(None=@n1.MaterialAttributes) {attributeGetTypes [
    #1
    #2
    #3
    #4
   ]}
  n3 = VectorParameter() {
   group=6_EmissiveColor
   parameterName=Emissive_Color
   sortPriority=0
   channelNames {a=Emissive, b="", g="", r=""}
   defaultValue=<rgba 0.3467 0.9131 1 0.6>
  }
  n4 = OneMinus(None=@n1.VertexColorR)
  n5 = Multiply(A=ComponentMask(None=@n1.RimRGBAMask) {a=true}, B=@n4)
  n6 = Multiply(A=Add(A=ComponentMask(None=@n1.RimRGBAMask) {r=true}, B=ComponentMask(None=@n1.RimRGBAMask) {g=true}), B=@n4)
  n7 = Fresnel(
   ExponentIn=ScalarParameter() {defaultValue=2, group=6_EmissiveColor, parameterName=FresnelValue}
   BaseReflectFractionIn=ScalarParameter() {defaultValue=-0.04, group=6_EmissiveColor}
   Normal=Transform(None=@n2.Normal)
  ) {baseReflectFraction=1}
  n8 = ScalarParameter() {defaultValue=2, group=6_EmissiveColor, parameterName=FresnelHighlight}
  n9 = OneMinus(None=TextureSample() {samplerType=SAMPLERTYPE_LinearColor, texture=/VehicleCosmetics/Wheels/Studded/Textures/T_Rim_Studded_Mask}.R)
  n10 = VertexColor()
  n11 = SetMaterialAttributes!(
   MaterialAttributes=@n2.MaterialAttributes
   Base Color=LinearInterpolate(A=LinearInterpolate(A=@n2.Base Color, B=@n3.RGB, Alpha=@n5), B=Multiply(A=@n8, B=@n2.Base Color), Alpha=Multiply(A=@n6, B=@n7))
   Emissive Color=LinearInterpolate(A=@n2.Emissive Color, B=@n3.Emissive, Alpha=@n5)
   Opacity Mask=Subtract(A=@n2.Opacity Mask, B=Multiply(A=@n9, B=@n10.B))
   Opacity=Multiply(
    A=@n4
    B=TextureSampleParameter2D() {parameterName=Heighmap, samplerType=SAMPLERTYPE_Masks, texture=/VehicleCosmetics/Textures/T_Black_Masks}.R
   )
  ) {attributeSetTypes [
    #1
    #2
    #3
    #5
   ]}
  MaterialFunctionCall(In=MaterialFunctionCall(Material=@n11) {materialFunction=&MF_BaseVehicle_Damage}.Result) {materialFunction=&MF_VehicleDeRez_Transition}
 ]
}
